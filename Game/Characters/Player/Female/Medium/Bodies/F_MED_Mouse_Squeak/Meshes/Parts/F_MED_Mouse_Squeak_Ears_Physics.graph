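# Physics asset for the ears part of the female medium Mouse Squeak player body, covering both ears and their earrings.
# Capsule bodies form a three-link chain per ear from the head, with tight swing and twist limits and twist locked at the tips.
# Two-link earring chains hang from those ears, two on the left (tip and middle segment) and one on the right (middle segment), all with twist locked and small swing limits.
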PhysicsAsset /Game/Characters/Player/Female/Medium/Bodies/F_MED_Mouse_Squeak/Meshes/Parts/F_MED_Mouse_Squeak_Ears_Physics
bodies {
 dyn_ear_01_l [{shapeType=Capsule, length=3.9693, radius=3.0248, shapeName=dyn_ear_01_l_capsule, center=<5 0 0>, rotation=<rot 90 0 0>}]
 dyn_ear_01_r [{shapeType=Capsule, length=3.6238, radius=3.7327, shapeName=dyn_ear_01_r_capsule, center=<-5 0 0>, rotation=<rot 90 0 0>}]
 dyn_ear_02_l [{shapeType=Capsule, length=5.8395, radius=3.2585, shapeName=dyn_ear_02_l_capsule, center=<5 0 0>, rotation=<rot 90 0 0>}]
 dyn_ear_02_r [{shapeType=Capsule, length=6.136, radius=3.2396, shapeName=dyn_ear_02_r_capsule, center=<-5 0 0>, rotation=<rot 90 0 0>}]
 dyn_ear_03_l [{shapeType=Capsule, length=1.9031, radius=2.9544, shapeName=dyn_ear_03_l_capsule, center=<5 0 0>, rotation=<rot 90 0 0>}]
 dyn_ear_03_r [{shapeType=Capsule, length=2.0815, radius=4.0883, shapeName=dyn_ear_03_r_capsule, center=<-5 0 0>, rotation=<rot 90 0 0>}]
 earring_01_r [{shapeType=Capsule, length=2, radius=3, shapeName=earring_b_01_l_capsule, center=<3 0 0>, rotation=<rot 90 0 0>}]
 earring_02_r [{shapeType=Capsule, length=2, radius=3, shapeName=earring_b_02_l_capsule, center=<3 0 0>, rotation=<rot 90 0 0>}]
 earring_a_01_l [{shapeType=Capsule, length=4, radius=1, shapeName=earring_a_01_l_capsule, center=<3 0 0>, rotation=<rot 90 0 0>}]
 earring_a_02_l [{shapeType=Capsule, length=4, radius=1, shapeName=earring_a_02_l_capsule, center=<3 0 0>, rotation=<rot 90 0 0>}]
 earring_b_01_l [{shapeType=Capsule, length=2, radius=4.5, shapeName=earring_b_01_l_capsule, center=<3 0 0>, rotation=<rot 90 0 0>}]
 earring_b_02_l [{shapeType=Capsule, length=2, radius=4.5, shapeName=earring_b_02_l_capsule, center=<3 0 0>, rotation=<rot 90 0 0>}]
 head [{shapeType=Capsule, length=1, radius=1}]
}
constraints [
 {bone1Name=dyn_ear_01_l, bone2Name=head, swing1LimitDegrees=5, swing2LimitDegrees=10, twistLimitDegrees=2.5}
 {bone1Name=dyn_ear_02_l, bone2Name=dyn_ear_01_l, swing1LimitDegrees=5, swing2LimitDegrees=5, twistLimitDegrees=2.5}
 {bone1Name=dyn_ear_03_l, bone2Name=dyn_ear_02_l, swing1LimitDegrees=5, swing2LimitDegrees=5, twistLimitDegrees=2.5, twistMotion=Locked}
 {bone1Name=dyn_ear_01_r, bone2Name=head, swing1LimitDegrees=5, swing2LimitDegrees=10, twistLimitDegrees=2.5}
 {bone1Name=dyn_ear_02_r, bone2Name=dyn_ear_01_r, swing1LimitDegrees=5, swing2LimitDegrees=5, twistLimitDegrees=2.5}
 {bone1Name=dyn_ear_03_r, bone2Name=dyn_ear_02_r, swing1LimitDegrees=5, swing2LimitDegrees=5, twistLimitDegrees=2.5, twistMotion=Locked}
 {bone1Name=earring_a_01_l, bone2Name=dyn_ear_03_l, swing1LimitDegrees=10, swing2LimitDegrees=5, twistLimitDegrees=45, twistMotion=Locked}
 {bone1Name=earring_a_02_l, bone2Name=earring_a_01_l, swing1LimitDegrees=10, swing2LimitDegrees=5, twistLimitDegrees=45, twistMotion=Locked}
 {bone1Name=earring_b_01_l, bone2Name=dyn_ear_02_l, swing1LimitDegrees=10, swing2LimitDegrees=10, twistLimitDegrees=45, twistMotion=Locked}
 {bone1Name=earring_b_02_l, bone2Name=earring_b_01_l, swing1LimitDegrees=10, swing2LimitDegrees=10, twistLimitDegrees=45, twistMotion=Locked}
 {bone1Name=earring_01_r, bone2Name=dyn_ear_02_r, swing1LimitDegrees=10, swing2LimitDegrees=10, twistLimitDegrees=45, twistMotion=Locked}
 {bone1Name=earring_02_r, bone2Name=earring_01_r, swing1LimitDegrees=10, swing2LimitDegrees=10, twistLimitDegrees=45, twistMotion=Locked}
]
