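# Physics asset for the whiskers part of the female medium Mouse Squeak player character body, defining bodies for simulated whisker motion.
# Four chains, two per side, each link four capsule bodies back to a small head capsule.
# Joints carry narrow swing limit angles that tighten toward the tips.
# Twist is locked along each chain except at the final joint, where it is free.
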PhysicsAsset /Game/Characters/Player/Female/Medium/Bodies/F_MED_Mouse_Squeak/Meshes/Parts/F_MED_Mouse_Squeak_Whiskers_Physics
bodies {
 head [{shapeType=Capsule, length=1, radius=1}]
 whisker_a_01_l [{shapeType=Capsule, length=4, radius=2, shapeName=whisker_a_01_l_capsule, center=<2 0 0>, rotation=<rot 90 0 0>}]
 whisker_a_01_r [{shapeType=Capsule, length=4, radius=2, shapeName=whisker_a_01_r_capsule, center=<2 0 0>, rotation=<rot 90 0 0>}]
 whisker_a_02_l [{shapeType=Capsule, length=4, radius=2, shapeName=whisker_a_02_l_capsule, center=<2 0 0>, rotation=<rot 90 0 0>}]
 whisker_a_02_r [{shapeType=Capsule, length=4, radius=2, shapeName=whisker_a_02_r_capsule, center=<2 0 0>, rotation=<rot 90 0 0>}]
 whisker_a_03_l [{shapeType=Capsule, length=4, radius=2, shapeName=whisker_a_03_l_capsule, center=<2 0 0>, rotation=<rot 90 0 0>}]
 whisker_a_03_r [{shapeType=Capsule, length=4, radius=2, shapeName=whisker_a_03_r_capsule, center=<2 0 0>, rotation=<rot 90 0 0>}]
 whisker_a_04_l [{shapeType=Capsule, length=4, radius=2, shapeName=whisker_a_01_l_capsule, center=<2 0 0>, rotation=<rot 90 0 0>}]
 whisker_a_04_r [{shapeType=Capsule, length=4, radius=2, shapeName=whisker_a_01_r_capsule, center=<2 0 0>, rotation=<rot 90 0 0>}]
 whisker_b_01_l [{shapeType=Capsule, length=4, radius=2, shapeName=whisker_b_01_l_capsule, center=<2 0 0>, rotation=<rot 90 0 0>}]
 whisker_b_01_r [{shapeType=Capsule, length=4, radius=2, shapeName=whisker_b_01_r_capsule, center=<2 0 0>, rotation=<rot 90 0 0>}]
 whisker_b_02_l [{shapeType=Capsule, length=4, radius=2, shapeName=whisker_b_02_l_capsule, center=<2 0 0>, rotation=<rot 90 0 0>}]
 whisker_b_02_r [{shapeType=Capsule, length=4, radius=2, shapeName=whisker_b_02_r_capsule, center=<2 0 0>, rotation=<rot 90 0 0>}]
 whisker_b_03_l [{shapeType=Capsule, length=4, radius=2, shapeName=whisker_b_03_l_capsule, center=<2 0 0>, rotation=<rot 90 0 0>}]
 whisker_b_03_r [{shapeType=Capsule, length=4, radius=2, shapeName=whisker_b_03_r_capsule, center=<2 0 0>, rotation=<rot 90 0 0>}]
 whisker_b_04_l [{shapeType=Capsule, length=4, radius=2, shapeName=whisker_a_01_r_capsule, center=<2 0 0>, rotation=<rot 90 0 0>}]
 whisker_b_04_r [{shapeType=Capsule, length=4, radius=2, shapeName=whisker_a_01_r_capsule, center=<2 0 0>, rotation=<rot 90 0 0>}]
}
constraints [
 {bone1Name=whisker_a_01_r, bone2Name=head, swing1LimitDegrees=5, swing2LimitDegrees=5, twistLimitDegrees=5, twistMotion=Locked}
 {bone1Name=whisker_a_02_r, bone2Name=whisker_a_01_r, swing1LimitDegrees=5, swing2LimitDegrees=5, twistLimitDegrees=5, twistMotion=Locked}
 {bone1Name=whisker_a_03_r, bone2Name=whisker_a_02_r, swing1LimitDegrees=2, swing2LimitDegrees=2, twistLimitDegrees=5, twistMotion=Locked}
 {bone1Name=whisker_b_01_l, bone2Name=head, swing1LimitDegrees=5, swing2LimitDegrees=5, twistLimitDegrees=5, twistMotion=Locked}
 {bone1Name=whisker_b_02_l, bone2Name=whisker_b_01_l, swing1LimitDegrees=5, swing2LimitDegrees=5, twistLimitDegrees=5, twistMotion=Locked}
 {bone1Name=whisker_b_03_l, bone2Name=whisker_b_02_l, swing1LimitDegrees=2, swing2LimitDegrees=2, twistLimitDegrees=5, twistMotion=Locked}
 {bone1Name=whisker_b_01_r, bone2Name=head, swing1LimitDegrees=5, swing2LimitDegrees=5, twistLimitDegrees=5, twistMotion=Locked}
 {bone1Name=whisker_b_02_r, bone2Name=whisker_b_01_r, swing1LimitDegrees=5, swing2LimitDegrees=5, twistLimitDegrees=5, twistMotion=Locked}
 {bone1Name=whisker_b_03_r, bone2Name=whisker_b_02_r, swing1LimitDegrees=2, swing2LimitDegrees=2, twistLimitDegrees=5, twistMotion=Locked}
 {bone1Name=whisker_a_01_l, bone2Name=head, swing1LimitDegrees=5, swing2LimitDegrees=5, twistLimitDegrees=5, twistMotion=Locked}
 {bone1Name=whisker_a_02_l, bone2Name=whisker_a_01_l, swing1LimitDegrees=5, swing2LimitDegrees=5, twistLimitDegrees=5, twistMotion=Locked}
 {bone1Name=whisker_a_03_l, bone2Name=whisker_a_02_l, swing1LimitDegrees=2, swing2LimitDegrees=2, twistLimitDegrees=5, twistMotion=Locked}
 {bone1Name=whisker_a_04_r, bone2Name=whisker_a_03_r, swing1LimitDegrees=2, swing2LimitDegrees=2, twistLimitDegrees=45, twistMotion=Free}
 {bone1Name=whisker_b_04_l, bone2Name=whisker_b_03_l, swing1LimitDegrees=2, swing2LimitDegrees=2, twistLimitDegrees=45, twistMotion=Free}
 {bone1Name=whisker_b_04_r, bone2Name=whisker_b_03_r, swing1LimitDegrees=2, swing2LimitDegrees=2, twistLimitDegrees=45, twistMotion=Free}
 {bone1Name=whisker_a_04_l, bone2Name=whisker_a_03_l, swing1LimitDegrees=2, swing2LimitDegrees=2, twistLimitDegrees=45, twistMotion=Free}
]
