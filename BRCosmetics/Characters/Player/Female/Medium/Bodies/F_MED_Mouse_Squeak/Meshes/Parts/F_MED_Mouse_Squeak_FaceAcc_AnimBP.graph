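# Post-copy animation blueprint for the female medium Mouse Squeak face accessory, adding secondary motion to ears and whiskers.
# Copies the pose from the attached parent mesh, then chains two base-bone-space RigidBody simulations: ears relative to head, whiskers relative to spine_05.
# Weighted constraints then drive each side's first whisker bone from head, inner cheek and nose bridge bones.
# All logic lives in the AnimGraph; the event graph holds only an empty update event.
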AnimBlueprint /BRCosmetics/Characters/Player/Female/Medium/Bodies/F_MED_Mouse_Squeak/Meshes/Parts/F_MED_Mouse_Squeak_FaceAcc_AnimBP
target_skeleton=/Game/Characters/Player/Female/Medium/Bodies/F_MED_Mouse_Squeak/Meshes/Parts/F_MED_Mouse_Squeak_FaceAcc_Skeleton
graphs [
 AnimGraph {nodes [
   n1 = RigidBody(ComponentPose=Local To Component(LocalPose=Copy Pose From Mesh() {bUseAttachedParent=true})) {
    bOverrideWorldGravity=true
    overridePhysicsAsset=/Game/Characters/Player/Female/Medium/Bodies/F_MED_Mouse_Squeak/Meshes/Parts/F_MED_Mouse_Squeak_Ears_Physics
    simulationSpace=BaseBoneSpace
    alphaBoolBlend {blendOption=Linear}
    baseBoneRef {boneName=head}
    componentAppliedLinearAccClamp=<1500 1500 1500>
    componentLinearAccScale=<0.2 0.2 0.3>
    componentLinearVelScale=<0.2 0.2 0.3>
    overrideWorldGravity {z=-980}
   }
   n2 = RigidBody(ComponentPose=@n1) {
    bOverrideWorldGravity=true
    overridePhysicsAsset=/Game/Characters/Player/Female/Medium/Bodies/F_MED_Mouse_Squeak/Meshes/Parts/F_MED_Mouse_Squeak_Whiskers_Physics
    simulationSpace=BaseBoneSpace
    alphaBoolBlend {blendOption=Linear}
    baseBoneRef {boneName=spine_05}
    componentAppliedLinearAccClamp=<2000 2000 2000>
    componentLinearAccScale=<0.3 0.3 0.35>
    componentLinearVelScale=<0.3 0.3 0.35>
    overrideWorldGravity {z=-980}
   }
   n3 = "Constraint - whisker_a_01_r (PPP)"!(ComponentPose=@n2, Alpha=1, ConstraintWeights_0=0.5, ConstraintWeights_1:float=0.46, ConstraintWeights_2:float=0.04) -> (Pose) {
    alphaBoolBlend {blendOption=Linear}
    boneToModify {boneName=whisker_a_01_r}
    constraintSetup [{targetBone {boneName=head}}, {targetBone {boneName=R_cheek_inner}}, {targetBone {boneName=C_nose_bridge}}]
    constraintWeights [1,1,1]
   }
   n4 = "Constraint - whisker_a_01_l (PPP)"!(ComponentPose=@n3, Alpha=1, ConstraintWeights_0=0.5, ConstraintWeights_1:float=0.46, ConstraintWeights_2:float=0.04) -> (Pose) {
    alphaBoolBlend {blendOption=Linear}
    boneToModify {boneName=whisker_a_01_l}
    constraintSetup [{targetBone {boneName=head}}, {targetBone {boneName=L_cheek_inner}}, {targetBone {boneName=C_nose_bridge}}]
    constraintWeights [1,1,1]
   }
   Output Pose(Result=Component To Local(ComponentPose=@n4)) {layerGroup=None}
  ]}
 EventGraph {nodes [Event BlueprintUpdateAnimation()]}
]
variables []
